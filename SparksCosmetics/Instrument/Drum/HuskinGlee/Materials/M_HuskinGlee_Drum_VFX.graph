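# Unlit VFX material for the HuskinGlee drum cosmetic, flagged for skeletal meshes, producing animated green plasma.
# A radially panned plasma sample offsets the UVs of a second plasma sample, which blends two greens into MF_FX_EmissiveAdjustments.
# Both plasma samples take derivatives from unmodified UVs, which avoids mip seams from radial mapping.
# A sine-pulsed foam and goo-dot UV distortion branch and a duplicate panner feed nothing shown.
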Material /SparksCosmetics/Instrument/Drum/HuskinGlee/Materials/M_HuskinGlee_Drum_VFX
graph {
 properties {
  bAutomaticallySetUsageInEditor=false
  bUsedWithMeshDeformer=true
  bUsedWithSkeletalMesh=true
  shadingModel=MSM_Unlit
  usedShadingModels=Unlit
 }
 nodes [
  n1 = TextureCoordinate()
  n2 = TextureSample() {samplerType=SAMPLERTYPE_Normal, texture=/FortWater/Textures/02_WaterFoam_Veroni_base_N}
  n3 = Multiply(A=AppendVector(A=@n2.R, B=@n2.R)) {constB=0.5}
  n4 = TextureSample() {texture=/BRCosmetics/Gadgets/Assets/VinderTech_GliderChute/Glider_AnglePatch/FX/T_AnglePatch_GooDots}
  n5 = Constant() {r=2}
  Add(
   A=Add(A=Multiply(A=@n1) {constB=2}) {constB=-1}
   B=Multiply(
    A=Multiply(A=@n3, B=@n4.G)
    B=MaterialFunctionCall(Speed=@n5) {materialFunction=&DebugTimeSine}.Result
   )
  )
  n6 = Constant3Vector() {constant=<rgba 0.0305 0.0729 0.0012 1>}
  n7 = Constant3Vector() {constant=<rgba 0.1264 0.3021 0.0049 1>}
  n8 = TextureCoordinate()
  n9 = TextureCoordinate()
  n10 = TextureSample!(
   UVs=Panner(Coordinate=MaterialFunctionCall() {materialFunction=&VectorToRadialValue}.Radial Coordinates) {speedY=-0.4}
   Tex
   "DDX(UVs)"=DDX(Value=@n9)
   "DDY(UVs)"=DDY(Value=@n9)
   Apply View MipBias
  ) {
   mipValueMode=TMVM_Derivative
   samplerType=SAMPLERTYPE_LinearGrayscale
   texture=/Game/Effects/Fort_Effects/Emotes/UltraLight/Texture/T_Plasma_D_Greyscale
  }
  n11 = TextureCoordinate()
  n12 = TextureSample(UVs=Multiply(A=Add(A=@n8, B=@n10.G)) {constB=0.3}, "DDX(UVs)"=DDX(Value=@n11), "DDY(UVs)"=DDY(Value=@n11)) {
   mipValueMode=TMVM_Derivative
   samplerType=SAMPLERTYPE_LinearGrayscale
   texture=&T_Plasma_D_Greyscale
  }
  MaterialFunctionCall(Emissive=Multiply(A=LinearInterpolate(A=@n6, B=@n7, Alpha=@n12.G)) {constB=2}) {materialFunction=&MF_FX_EmissiveAdjustments}
  Panner(Coordinate=MaterialFunctionCall() {materialFunction=&VectorToRadialValue}.Radial Coordinates) {speedY=-0.4}
 ]
}
